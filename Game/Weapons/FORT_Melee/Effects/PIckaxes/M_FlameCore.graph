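# Translucent, unlit, two-sided material for a flame core in melee pickaxe effects, usable on Cascade and Niagara mesh particles.
# Particle color RGB passes through MF_FX_EmissiveAdjustments to form the glow.
# Two time-driven sine waves of differing speed and frequency, phased by UV V, are summed along the transformed local -Y axis.
# The result is scaled by V, forming a wobble vector that is zero at V=0 and weighted toward higher V.
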Material /Game/Weapons/FORT_Melee/Effects/PIckaxes/M_FlameCore
graph {
 properties {
  bUsedWithMeshParticles=true
  bUsedWithNiagaraMeshParticles=true
  blendMode=BLEND_Translucent
  pixelDepthOffsetMode=PDOM_Legacy
  shadingModel=MSM_Unlit
  twoSided=true
  usedShadingModels=Unlit
 }
 nodes [
  MaterialFunctionCall(Emissive=ParticleColor().RGB) {materialFunction=&MF_FX_EmissiveAdjustments}
  n1 = Constant3Vector() {constant=<rgba 0 -1 0 0>}
  n2 = Multiply(A=Time()) {constB=3.123}
  n3 = Multiply(A=ComponentMask(None=TextureCoordinate()) {g=true}) {constB=1.7}
  n4 = Multiply(A=Sine(None=Add(A=@n2, B=@n3))) {constB=1.3}
  n5 = Multiply(A=Time()) {constB=5.713}
  n6 = Multiply(A=ComponentMask(None=TextureCoordinate()) {g=true}) {constB=5.871}
  n7 = Multiply(A=Sine(None=Add(A=@n5, B=@n6))) {constB=1.574}
  Multiply(
   A=ComponentMask(None=TextureCoordinate()) {g=true}
   B=Multiply(A=Transform(None=@n1) {transformSourceType=TRANSFORMSOURCE_Local}, B=Add(A=@n4, B=@n7))
  )
 ]
}
