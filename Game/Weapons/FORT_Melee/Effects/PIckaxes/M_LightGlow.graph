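# Additive unlit glow material for mesh particles in pickaxe effects, tinted by particle color.
# Emissive runs through MF_FX_EmissiveAdjustments, with an optional brightness boost for low and medium quality or forward and mobile paths.
# Alpha flickers from two summed sines unless simple glow is set, with optional depth fade and height-based bottom fade.
# Optional vertex-normal offset scales by a dynamic parameter; scalar-driven lerps toggle features instead of static switches.
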Material /Game/Weapons/FORT_Melee/Effects/PIckaxes/M_LightGlow
graph {
 properties {
  bUsedWithMeshParticles=true
  bUsedWithNiagaraMeshParticles=true
  blendMode=BLEND_Additive
  pixelDepthOffsetMode=PDOM_Legacy
  shadingModel=MSM_Unlit
  usedShadingModels=Unlit
 }
 nodes [
  LinearInterpolate(
   B=Multiply(A=VertexNormalWS(), B=DynamicParameter() {paramNames [Param1,Param2,Param3,WPOScale]}.WPOScale)
   Alpha=ScalarParameter() {parameterName=Use WPO Scale}
  )
  n1 = ParticleColor()
  n2 = Multiply(A=@n1.RGB, B=ScalarParameter() {defaultValue=1.5, parameterName=Brightness on low}) {constB=1.5}
  n3 = QualitySwitch(Default=@n1.RGB, Low=@n2, High=@n1.RGB, Medium=@n2, Epic=@n1.RGB)
  MaterialFunctionCall(Emissive=LinearInterpolate(
    A=@n1.RGB
    B=ShadingPathSwitch(Default=@n3, Deferred=@n3, Forward=@n2, Mobile=@n2)
    Alpha=ScalarParameter() {parameterName=Use different brightness on low}
   )) {materialFunction=&MF_FX_EmissiveAdjustments}
  n4 = Multiply(A=Time()) {constB=3.123}
  n5 = Multiply(A=Time()) {constB=5.713}
  n6 = Add(A=Multiply(A=Sine(None=@n4)) {constB=0.5}, B=Multiply(A=Sine(None=@n5)) {constB=0.5})
  n7 = LinearInterpolate(A=Multiply(A=@n1.A, B=Add(A=@n6)), B=@n1.A, Alpha=ScalarParameter() {parameterName=Use_Simple_Glow})
  n8 = ScalarParameter() {defaultValue=5, parameterName=FadeDistance}
  n9 = Add(A=WorldPosition().XYZ) {constB=50}
  n10 = ParticlePositionWS()
  n11 = Divide(A=ComponentMask(None=Subtract(A=@n9, B=@n10)) {b=true}) {constB=150}
  Multiply(
   A=LinearInterpolate(A=@n7, B=DepthFade(Opacity=@n7, FadeDistance=@n8), Alpha=ScalarParameter() {parameterName=Turn on the depthfade})
   B=LinearInterpolate(A=Constant() {r=1}, B=Clamp(None=@n11), Alpha=ScalarParameter() {parameterName=Use_bottom_fade})
  )
  ObjectPositionWS()
  Sine()
 ]
}
