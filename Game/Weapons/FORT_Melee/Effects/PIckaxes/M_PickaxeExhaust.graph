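# Unlit translucent mesh-particle material for an exhaust effect on pickaxes.
# A panning noise texture varies a Fresnel exponent, and the result is multiplied by vertex red, particle color and an inverted-V gradient.
# An offset chain, weighted by UV V, combines noise along vertex normals with two time-driven sine waves along local -Y.
# Other branches build blue color parameters and a highlight from a clamped local-space camera direction dotted with the pixel normal.
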Material /Game/Weapons/FORT_Melee/Effects/PIckaxes/M_PickaxeExhaust
graph {
 properties {
  bUsedWithMeshParticles=true
  bUsedWithNiagaraMeshParticles=true
  blendMode=BLEND_Translucent
  pixelDepthOffsetMode=PDOM_Legacy
  shadingModel=MSM_Unlit
  usedShadingModels=Unlit
 }
 nodes [
  n1 = VectorParameter() {defaultValue=<rgba 0 0.5 3 0>}
  n2 = MaterialFunctionCall(Float3=Transform(None=MaterialFunctionCall() {materialFunction=&CameraDirectionVector}.Result) {transformSourceType=TRANSFORMSOURCE_World, transformType=TRANSFORM_Local}) {materialFunction=&BreakOutFloat3Components}
  n3 = Clamp(None=@n2.G) {maxDefault=0, minDefault=-1}
  n4 = DotProduct(
   A=Normalize(VectorInput=MaterialFunctionCall(X=@n2.R, Y=@n3, Z=@n2.R) {materialFunction=&MakeFloat3}.Result)
   B=Transform(None=PixelNormalWS()) {transformSourceType=TRANSFORMSOURCE_World, transformType=TRANSFORM_Local}
  )
  LinearInterpolate(
   A=@n1.RGB
   B=VectorParameter() {defaultValue=<rgba 0.9549 5.6393 10 0>}.RGB
   Alpha=Multiply(A=Power(Base=@n4) {constExponent=32}, B=VertexColor().R)
  )
  MaterialFunctionCall(Emissive=@n1.RGB) {materialFunction=&MF_FX_EmissiveAdjustments}
  n5 = TextureSample(UVs=Panner() {bFractionalPart=true, speedY=-2.5}) {texture=&T_NoisePack_01}
  n6 = Multiply(A=@n5.R, B=VertexNormalWS())
  n7 = Transform(None=Constant3Vector() {constant=<rgba 0 -1 0 0>}) {transformSourceType=TRANSFORMSOURCE_Local}
  n8 = TextureCoordinate()
  n9 = Add(A=Multiply(A=Time()) {constB=3.123}, B=Multiply(A=ComponentMask(None=@n8) {g=true}) {constB=1.7})
  n10 = TextureCoordinate()
  n11 = Add(A=Multiply(A=Time()) {constB=5.713}, B=Multiply(A=ComponentMask(None=@n10) {g=true}) {constB=5.871})
  n12 = Add(A=Multiply(A=Sine(None=@n9)) {constB=1.3}, B=Multiply(A=Sine(None=@n11)) {constB=1.574})
  Multiply(A=ComponentMask(None=TextureCoordinate()) {g=true}, B=Add(A=Multiply(A=@n6) {constB=4}, B=Multiply(A=@n7, B=@n12)))
  Multiply(A=@n4) {constB=2}
  Reroute(None=@n1.RGB)
  n13 = Constant() {r=3}
  n14 = Constant() {r=5}
  n15 = Multiply(
   A=MaterialFunctionCall(Exponent=LinearInterpolate(A=@n13, B=@n14, Alpha=@n5.R)) {materialFunction=&MF_Fresnel_FX}.Fresnel
   B=VertexColor().R
  )
  n16 = ParticleColor()
  n17 = TextureCoordinate()
  n18 = Power(Base=OneMinus(None=ComponentMask(None=@n17) {g=true}))
  Reroute(None=Multiply(A=Multiply(A=@n15, B=@n16.RGB), B=Multiply(A=@n18) {constB=2}))
 ]
}
