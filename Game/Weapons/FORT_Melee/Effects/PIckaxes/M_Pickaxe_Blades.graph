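# Additive, unlit, two-sided mesh-particle material that gives pickaxe blades an animated energy glow, blue to cyan by default.
# Screen-aligned UVs are offset by the actor's clip-space position and scaled by its depth, anchoring the pattern to the actor.
# Two panning laser-sword glow samples are multiplied and sharpened to blend dark and light color parameters.
# The same noise, particle alpha, Fresnel, red vertex color and depth fade form a faint intensity mask.
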Material /Game/Weapons/FORT_Melee/Effects/PIckaxes/M_Pickaxe_Blades
graph {
 properties {
  bUsedWithMeshParticles=true
  blendMode=BLEND_Additive
  pixelDepthOffsetMode=PDOM_Legacy
  shadingModel=MSM_Unlit
  twoSided=true
  usedShadingModels=Unlit
 }
 nodes [
  n1 = Subtract(A=MaterialFunctionCall() {materialFunction=&ScreenAlignedUVs}."X 100%, Y 100%") {constB=0.5}
  MaterialExpressionActorPositionWS_1 = ActorPositionWS()
  n2 = Custom!(In=@MaterialExpressionActorPositionWS_1) {
   code="mul(float4(In.xyz,1),LWCToFloat(ResolvedView.WorldToClip))"
   desc="to clip space\n*** Original source before expression upgrade ***\nmul(float4(In.xyz,1),View.WorldToClip)"
   outputType=CMOT_Float4
   inputs [{inputName=In, input {expression=&self:MaterialExpressionActorPositionWS_1}}]
  }
  n3 = Divide(A=ComponentMask(None=@n2) {g=true, r=true}, B=Multiply(A=ComponentMask(None=@n2) {a=true}) {constB=2})
  n4 = MaterialFunctionCall(Float2=Multiply(A=Multiply(A=@n3))) {materialFunction=&BreakOutFloat2Components}
  n5 = Constant()
  n6 = MaterialFunctionCall(X=@n4.R, Y=@n5) {materialFunction=&MakeFloat2}
  n7 = Multiply(
   A=Add(
    A=Subtract(A=@n1, B=@n6.Result)
    B=MaterialFunctionCall(X=@n5, Y=@n4.G) {materialFunction=&MakeFloat2}.Result
   ) {constB=0}
   B=Divide(A=ComponentMask(None=@n2) {a=true}, B=ScalarParameter() {defaultValue=128, parameterName=uvScaleFactor})
  )
  n8 = Multiply(A=@n7, B=Constant2Vector() {g=1.75, r=5.81}) {constB=4}
  n9 = Multiply(A=@n7, B=Constant2Vector() {g=1.325, r=4.505}) {constB=1.7}
  n10 = Multiply(
   A=TextureSample(UVs=Panner(Coordinate=@n8) {bFractionalPart=true, speedX=-0.25, speedY=0.8155}) {texture=/Game/Weapons/FORT_Melee/Effects/LaserSwordFounders/T_LaserSwordGlow}.RGB
   B=TextureSample(UVs=Panner(Coordinate=@n9) {bFractionalPart=true, speedX=0.2065, speedY=0.605}) {texture=&T_LaserSwordGlow}.RGB
  )
  MaterialFunctionCall(Emissive=LinearInterpolate(
    A=VectorParameter() {parameterName=DarkCol, defaultValue=<rgba 0.0757 0.2293 1 0>}.RGB
    B=VectorParameter() {parameterName=LightCol, defaultValue=<rgba 0.0955 1.0118 1.5 0>}.RGB
    Alpha=Power(Base=@n10) {constExponent=4}
   )) {materialFunction=&MF_FX_EmissiveAdjustments}
  n11 = Multiply(
   A=Multiply(A=Power(Base=@n10) {constExponent=3}, B=ParticleColor().A)
   B=MaterialFunctionCall(Exponent=Constant() {r=16}) {materialFunction=&MF_Fresnel_FX}.Fresnel
  )
  n12 = VertexColor()
  n13 = Constant() {r=12}
  Multiply(A=Multiply(A=Multiply(A=@n11, B=@n12.R), B=DepthFade(FadeDistance=@n13))) {constB=0.1}
 ]
}
